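# Niagara system for the ToonPsplus vehicle boost, a blue stylized exhaust trail emitted from filtered sockets on the car's skeletal mesh.
# Layered ribbons with tapering fill and flame widths form the jet, while growing SubUV sprites add puffs.
# A star emitter runs two spawn groups with different cone motion, layering outlined, multi-renderer sprites.
# Painted-variant overrides drive colors, and game-mode switches adjust spawn rates and velocities.
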NiagaraSystem /VehicleCosmetics/Boosts/Boost_ToonPsplus/FX/NS_Boost_ToonPsplus
dependencies {
 usedDataInterfaces [
  /Script/Niagara.NiagaraDataInterfaceArrayDistributionInt
  &NiagaraDataInterfaceCurve
  &NiagaraDataInterfaceArrayFloat
  &NiagaraDataInterfaceColorCurve
  &NiagaraDataInterfaceSpriteRendererInfo
  &NiagaraDataInterfaceArrayFloat3
  &NiagaraDataInterfaceSkeletalMesh
 ]
 usedDynamicInputs []
 usedModules [
  &self:SystemSpawnScript.SystemScriptSource.SystemScriptGraph.NiagaraNodeAssignment_0.SetVariables_#1
  &self:SystemSpawnScript.SystemScriptSource.SystemScriptGraph.NiagaraNodeAssignment_1.SetVariables_#2
  &self:SystemSpawnScript.SystemScriptSource.SystemScriptGraph.NiagaraNodeAssignment_2.SetVariables_#3
  &SystemState
 ]
}
emitters {
 NE_VC_Boost_Ribbon_Parent001 {
  emitterUpdateScript [
   &EmitterState()
   &VC_FilteredSocketCount()
   &VC_RateQualityHandler(High Spawn Rate=60, Low Spawn Rate=30, Use Custom Rate?=true)
   &VC_SpawnRibbon_v1(Number of Ribbons=$Emitter.PortCount, SpawnRate=$Emitter.Spawn Rate)
  ]
  particleSpawnScript [
   &VC_PaintedVariantsOverride(
    DefaultColor=<rgba 0 0.153 0.34 3>
    R1 DefaultColor=<rgba 0.1239 0.8219 1.8 5>
    R2 DefaultColor:LinearColor=<rgba 0.0545 0.2752 0.5449 2>
    Use R1 Painted Overrides=true
    Use R2 Painted Overrides=true
   )
   &VC_InitializeRibbon_v1(
    Color=$Particles.Color
    Custom Offset=true
    Enable Skeletal Position=true
    Lifetime=0.35
    Local Vector Offset=&AddVectorToPosition(Position=$User.DTCarOffset, Vector=<0 0 0>)
    Offset Amount=1
    Ribbon Width=100
   )
   &AddVelocity(Velocity=&VC_Mask_Vectors(Coro Value:Vector3f=<-333 0 0>, FrontEnd Value:Vector3f=<-800 0 0>, FN Value:Vector3f=<-1200 0 0>)) {module=AddVelocity001}
   /SetVariables(Particles.WidthRibbon:NiagaraFloat=&SelectFloatFromArray(Direct Array Index=$Particles.SpawnGroup, Float Selection Array=DataInterface() {propertyValues {FloatData [-120,120]}})) {isSetParameters=true}
   /VehicleCosmetics/FX/VC_FXParents/Modules/DLR_ProductPreviews(GameModeID:NiagaraInt32=$System.GameModeID, bPreviewVehicleSpeed:NiagaraBool=$System.bPreviewVehicleSpeed)
  ]
  particleUpdateScript [
   &ParticleState()
   &InheritSourceMovement(ApplyForce=false, ApplyPosition=true, ApplyPositionFalloff=false)
   &SolveForcesAndVelocity(Clamp Velocity=false, Limit Acceleration=false)
   /SetVariables(Particles.Flames:NiagaraPosition=&AddVectorToPosition(
     Position=$Particles.Position
     Vector=/VehicleCosmetics/FX/Niagara/Modules/VC_ConvertVectorToWorld(Vector:Vector3f=&SelectVectorFromArray(Direct Array Index=$Particles.SpawnGroup, Vector Selection Array=DataInterface() {propertyValues {FloatData [<0 12 0>, <0 -12 0>]}}))
    )) {isSetParameters=true}
   /SetVariables(
    Particles.FillRibbonWidth:NiagaraFloat=&FloatFromCurve(FloatCurve=DataInterface() {propertyValues {Curve {keys [<at 0.5248 1>,<at 0.7154 0>]}}}, Scale Curve=0.82)
    Particles.FlameRibbonWidth:NiagaraFloat=&Multiply_Float(
     A=$Particles.WidthRibbon
     B=&FloatFromCurve(FloatCurve=DataInterface() {propertyValues {Curve {keys [<at 0.4626 0.9475>,<at 0.9347 0>]}}})
    )
   ) {isSetParameters=true}
   &VC_UpdateRibbon_v1(Enable Skeletal Position=true)
  ]
  rendererClasses [&NiagaraRibbonRendererProperties]
  renderers [
   {rendererClass=&NiagaraRibbonRendererProperties}
   {rendererClass=&NiagaraRibbonRendererProperties, rendererIndex=1}
   {rendererClass=&NiagaraRibbonRendererProperties, rendererIndex=2}
  ]
 }
 NE_VC_Boost_Sprite_Parent {
  emitterUpdateScript [
   &EmitterState()
   &VC_FilteredSocketCount()
   &SpawnRate(
    SpawnRate=&Multiply_Float(
     A=32
     B=/VehicleCosmetics/FX/Niagara/Modules/VC_Mask_Floats(Coro Value:NiagaraFloat=1, FrontEnd Value:NiagaraFloat=0.5, FN Value:NiagaraFloat=0.5)
    )
    Use Spawn Probability=false
   )
  ]
  particleSpawnScript [
   &VC_PaintedVariantsOverride() {enabled=false}
   &InitializeParticle(
    Color Mode=::NewEnumerator0
    Lifetime=0.55
    Sprite Rotation Mode=::NewEnumerator2
    Sprite Size Mode=::NewEnumerator3
    Uniform Sprite Size=40
    UsePositionOffset=false
    Write Material Random=true
   )
   &PartitionParticles(Number Of Partitions=$Emitter.PortCount)
   &SkeletalMeshLocation(
    Bone / Socket Index=$Particles.PartitionParticles.Partition
    Bone / Socket Sampling Mode=::NewEnumerator5
    MaskSpawnGroup=false
    Mesh Sampling Type=::NewEnumerator6
    Sampled Position Offset=&AddVectorToPosition(Position=<-10 0 0>, Vector=Linked() {linkedVariable {name=User.DTCarOffset, type=NiagaraPosition}})
    Skeletal Mesh=$System.SkeletalMesh
   )
   &AddVelocity(Velocity=&VC_Mask_Vectors(Coro Value:Vector3f=<-400 0 0>, FrontEnd Value:Vector3f=<-150 0 0>, FN Value:Vector3f=<-1200 0 0>))
   &DLR_ProductPreviews(GameModeID:NiagaraInt32=$System.GameModeID, bPreviewVehicleSpeed:NiagaraBool=$System.bPreviewVehicleSpeed)
  ]
  particleUpdateScript [
   &ParticleState()
   &VC_PaintedVariantsOverride(
    Cobalt=<rgba 0 0.0185 1 1>
    DefaultColor=&ColorFromCurve(ColorCurve=DataInterface() {propertyValues {
       AlphaCurve {keys [<at 0 1 RCIM_Cubic>]}
       BlueCurve {keys [<at 0 0.48 RCIM_Cubic>,<at 1 0.7014>]}
       GreenCurve {keys [<at 0 0.2532 RCIM_Cubic>,<at 1 0.37>]}
       RedCurve {keys [<at 0 0.0264 RCIM_Cubic>,<at 1 0.0386>]}
      }})
   ) {module=VC_PaintedVariantsOverride001}
   &ScaleSpriteSize(
    Uniform Curve Scale=1.35
    Uniform Curve Sprite Scale=DataInterface() {propertyValues {Curve {keys [{tangentMode=RCTM_User, time=0, value=1}, {tangentMode=RCTM_User, time=1, value=2.5}]}}}
   )
   &SubUVAnimation(
    "Sprite Renderer "=DataInterface() {propertyValues {SpriteRenderer=&self:NE_VC_Boost_Sprite_Parent.NiagaraSpriteRendererProperties_3}}
    SubUV Animation Mode:ENiagara_SubUVLookupModeV2=::NewEnumerator1
    SubUV Curve=DataInterface() {propertyValues {Curve {
       defaultValue=0
       keys [
        {arriveTangent=0.5133, interpMode=RCIM_Cubic, leaveTangent=0.5133, tangentMode=RCTM_User, time=0, value=0}
        {interpMode=RCIM_Cubic, tangentMode=RCTM_User, time=0.85, value=1}
       ]
      }}}
    UseEndFrame=false
    UseStartFrame=false
   )
   &InheritSourceMovement(ApplyForce=false, ApplyPosition=true, ApplyPositionFalloff=false) {enabled=false}
   &SolveForcesAndVelocity(Clamp Velocity=false, Limit Acceleration=false)
  ]
  rendererClasses [&NiagaraSpriteRendererProperties]
  renderers [{rendererClass=&NiagaraSpriteRendererProperties}]
 }
 NE_VC_Boost_Sprite_Stars {
  emitterUpdateScript [
   &EmitterState()
   &VC_FilteredSocketCount()
   &SpawnRate(
    SpawnRate=&Multiply_Float(
     A=60
     B=&VC_Mask_Floats(Coro Value:NiagaraFloat=1, FrontEnd Value:NiagaraFloat=0.5, FN Value:NiagaraFloat=0.5)
    )
    Use Spawn Probability=false
   )
   &SpawnRate(
    SpawnGroup=1
    SpawnRate=&Multiply_Float(
     A=80
     B=&VC_Mask_Floats(Coro Value:NiagaraFloat=1, FrontEnd Value:NiagaraFloat=0.5, FN Value:NiagaraFloat=0.5)
    )
    Use Spawn Probability=false
   ) {module=SpawnRate001}
  ]
  particleSpawnScript [
   /SetVariables(Particles.VisibilityTag=$Emitter.SpawnGroup) {isSetParameters=true}
   &VC_PaintedVariantsOverride(
    R2 Cobalt:LinearColor=<rgba 0 0.0185 1 1>
    R2 DefaultColor:LinearColor=&MaskLinearColor(
     Mask Source=1
     Mask Parameter=/Niagara/DynamicInputs/Arrays/SelectIntFromWeightedDistributionArray(
      Array Sampling Mode:ENiagara_LifetimeMode=::NewEnumerator1
      Direct Array Index:NiagaraInt32=?
      Direct Array Mode:ENiagara_WrapClamp=::NewEnumerator0
      Evaluation Type:ENiagaraRandomnessEvaluation=::NewEnumerator0
      Recalculate Random Each Loop:NiagaraBool=false
      Weighted Distribution Int Array:NiagaraDataInterfaceArrayDistributionInt=DataInterface() {propertyValues {ArrayData [{value=0, weight=1}, {value=1, weight=0.5}]}}
      Randomness Mode:ENiagaraRandomnessMode=::NewEnumerator0
      Random Seed:NiagaraInt32=0
      Fixed Random Seed:NiagaraBool=false
     )
     Mask=0
     Masked Linear Color=<rgba 0.9823 0.8796 0.0513 1>
     Passthrough Linear Color=<rgba 1 1 1 1>
    )
    R2 Gold:LinearColor=<rgba 0.9823 0.8796 0.0513 1>
    R3 DefaultColor:LinearColor=<rgba 0 0 0 1>
    Use R2 Painted Overrides=true
    Use R3 Painted Overrides=true
   )
   &InitializeParticle(
    Color Mode=::NewEnumerator0
    Lifetime=&MergeHelper_FloatSwitch(
     0=1
     1=&RandomRangeFloat(Minimum=0.8, Maximum=1.5)
     2=?
     3=?
     Index=$Emitter.SpawnGroup
     Num=::NewEnumerator0
    )
    Sprite Rotation Angle=&MergeHelper_FloatSwitch(
     0=&RandomRangeFloat(Minimum=0, Maximum=1)
     1=&RandomRangeFloat(Minimum=0.05, Maximum=0.2)
     2=?
     3=?
     Index=$Emitter.SpawnGroup
     Num=::NewEnumerator0
    )
    Sprite Rotation Mode=::NewEnumerator1
    Sprite Size Mode=::NewEnumerator3
    Uniform Sprite Size=140
    UsePositionOffset=false
    Write Material Random=true
   )
   &PartitionParticles(Number Of Partitions=$Emitter.PortCount)
   &SkeletalMeshLocation(
    Bone / Socket Index=$Particles.PartitionParticles.Partition
    Bone / Socket Sampling Mode=::NewEnumerator5
    MaskSpawnGroup=false
    Mesh Sampling Type=::NewEnumerator6
    Sampled Position Offset=&AddVectorToPosition(Position=<-20 0 0>, Vector=Linked() {linkedVariable {name=User.DTCarOffset, type=NiagaraPosition}})
    Skeletal Mesh=$System.SkeletalMesh
   )
   &ShapeLocation(
    MaskSpawnGroup=false
    Override Seed=false
    Sphere Radius=&MergeHelper_FloatSwitch(0=25, 1=0, 2=?, 3=?, Index=$Emitter.SpawnGroup, Num=::NewEnumerator0)
    UseOwnerScale=true
   )
   &AddVelocity(
    Cone Angle=&MergeHelper_FloatSwitch(0=35, 1=42, 2=?, 3=?, Index=$Emitter.SpawnGroup, Num=::NewEnumerator0)
    Distribution Along Cone Axis:NiagaraFloat=&MergeHelper_FloatSwitch(0=0.5, 1=1, 2=?, 3=?, Index=$Emitter.SpawnGroup, Num=::NewEnumerator0)
    Inner Cone Angle:NiagaraFloat=&MergeHelper_FloatSwitch(0=30, 1=0, 2=?, 3=?, Index=$Emitter.SpawnGroup, Num=::NewEnumerator0)
    Module.UseConeFalloff=true
    Module.UseVelDistribution=true
    Override Seed=false
    Speed Falloff From Cone Axis:NiagaraFloat=&MergeHelper_FloatSwitch(0=0.5, 1=1, 2=?, 3=?, Index=$Emitter.SpawnGroup, Num=::NewEnumerator0)
    Velocity Mode=::NewEnumerator2
    Velocity Speed=&MergeHelper_FloatSwitch(
     0=-700
     1=&RandomRangeFloat(
      Minimum=&MergeHelper_FloatSwitch(0=-700, 1=-700, 2=-700, 3=-350, Index=$User.GameModeID, Num=::NewEnumerator2)
      Maximum=&MergeHelper_FloatSwitch(0=-950, 1=-950, 2=-950, 3=-570, Index=$User.GameModeID, Num=::NewEnumerator2)
     )
     2=?
     3=?
     Index=$Emitter.SpawnGroup
     Num=::NewEnumerator0
    )
   )
   &DLR_ProductPreviews(GameModeID:NiagaraInt32=$System.GameModeID, bPreviewVehicleSpeed:NiagaraBool=$System.bPreviewVehicleSpeed)
  ]
  particleUpdateScript [
   &ParticleState()
   &ScaleSpriteSize(Uniform Curve Sprite Scale=DataInterface() {propertyValues {Curve {
       defaultValue=1
       keys [
        <at 0 0 RCIM_Cubic>
        {arriveTangent=2.1099, interpMode=RCIM_Cubic, leaveTangent=2.1099, tangentMode=RCTM_User, time=0.2, value=1}
        {
         arriveTangent=0.6028
         arriveTangentWeight=0.1776
         interpMode=RCIM_Cubic
         leaveTangent=0.6028
         leaveTangentWeight=0.1776
         tangentMode=RCTM_User
         tangentWeightMode=RCTWM_WeightedBoth
         time=1
         value=1.5
        }
       ]
      }}})
   &SpriteRotationRate(Rotation Rate=&MergeHelper_FloatSwitch(0=100, 1=0, 2=?, 3=?, Index=&SpawnGroup(), Num=::NewEnumerator0))
   &AccelerationForce(Acceleration=&MergeHelper_VectorSwitch(
     0=<0 0 0>
     1=&MakeVector(
      X=0
      Y=&FloatFromCurve(
       FloatCurve=DataInterface() {propertyValues {Curve {defaultValue=0, keys [<at 0 0 RCIM_Constant>,<at 0.4 1 RCIM_Constant>]}}}
       Scale Curve=&SelectFloatFromArray(
        Array Sampling Mode=::NewEnumerator0
        Direct Array Index=?
        Float Selection Array=DataInterface() {propertyValues {FloatData [-800,800]}}
        Random Seed=0
       )
      )
      Z=0
     )
     2=?
     3=?
     Index=&SpawnGroup()
     Num=::NewEnumerator0
    ))
   &Drag(
    Drag=&MergeHelper_FloatSwitch(0=1, 1=1.2, 2=?, 3=?, Index=&SpawnGroup(), Num=::NewEnumerator0)
    Use Linear Drag=true
    Use Rotational Drag=false
   )
   &InheritSourceMovement(ApplyForce=false, ApplyPosition=true, ApplyPositionFalloff=false)
   &SolveForcesAndVelocity(Clamp Velocity=false, Limit Acceleration=false)
   &ScaleColor(
    Scale Alpha=&FloatFromCurve(FloatCurve=DataInterface() {propertyValues {Curve {keys [<at 0.05 0>,<at 0.2 1>]}}})
    Scale RGB=&MakeVectorFromLinearColorRGB(Color=&MaskLinearColor(
      Mask Source=1
      Mask Parameter=&RandomRangeInt(Minimum=0, Maximum=1)
      Mask=0
      Masked Linear Color=&ColorFromCurve(ColorCurve=DataInterface() {propertyValues {
         AlphaCurve {keys []}
         BlueCurve {keys [<at 0 0.4 RCIM_Cubic>,<at 0.7751 0.12>]}
         GreenCurve {keys [<at 0 0.1595 RCIM_Cubic>,<at 0.7751 0.041>]}
         RedCurve {keys [<at 0 0.0172 RCIM_Cubic>,<at 0.7751 0.024>]}
        }})
      Passthrough Linear Color=&ColorFromCurve(ColorCurve=DataInterface() {propertyValues {
         AlphaCurve {keys []}
         BlueCurve {keys [<at 0 0.4 RCIM_Cubic>,<at 0.7934 0.2>]}
         GreenCurve {keys [<at 0 0.1595 RCIM_Cubic>,<at 0.7934 0.0594>]}
         RedCurve {keys [<at 0 0.0172 RCIM_Cubic>,<at 0.7934 0.0026>]}
        }})
     ))
    ScaleA=true
    ScaleRGB=true
   )
   /SetVariables(
    Particles.StarBackSize:Vector2f=&Multiply_Vector2D_ByFloat(
     Float=&FloatFromCurve(
      FloatCurve=DataInterface() {propertyValues {Curve {
         defaultValue=1
         keys [
          {arriveTangent=4.4693, interpMode=RCIM_Cubic, leaveTangent=4.4693, tangentMode=RCTM_User, time=-0.04, value=0}
          {arriveTangent=0.0366, interpMode=RCIM_Cubic, leaveTangent=0.0366, leaveTangentWeight=0.0836, tangentMode=RCTM_User, time=0.75, value=1}
          {arriveTangent=-9.0725, arriveTangentWeight=0.7606, interpMode=RCIM_Cubic, leaveTangent=-9.0725, tangentMode=RCTM_User, time=0.95, value=0}
         ]
        }}}
      Scale Curve=1.3
     )
     Vector2D=<32 32>
    )
    Particles.StarFrontSize:Vector2f=&Multiply_Vector2D_ByFloat(
     Float=&FloatFromCurve(FloatCurve=DataInterface() {propertyValues {Curve {
         defaultValue=1
         keys [
          {arriveTangent=4.4693, interpMode=RCIM_Cubic, leaveTangent=4.4693, tangentMode=RCTM_User, time=-0.05, value=0}
          {arriveTangent=0.0366, interpMode=RCIM_Cubic, leaveTangent=0.0366, leaveTangentWeight=0.0836, tangentMode=RCTM_User, time=0.75, value=1}
          {arriveTangent=-9.0725, arriveTangentWeight=0.7606, interpMode=RCIM_Cubic, leaveTangent=-9.0725, tangentMode=RCTM_User, time=0.95, value=0}
         ]
        }}})
     Vector2D=<30 30>
    )
   ) {isSetParameters=true}
   /SetVariables(
    Particles.OutlineColor:LinearColor=&MakeLinearColorFromVectorAndFloat(
     "Vector (RGB)"=&MakeVectorFromLinearColorRGB(Color=&ColorFromCurve(ColorCurve=DataInterface() {propertyValues {
         AlphaCurve {keys []}
         BlueCurve {keys [<at 0 0.16 RCIM_Cubic>,<at 0.6172 0.1>]}
         GreenCurve {keys [<at 0 0.0376 RCIM_Cubic>,<at 0.6172 0.0235>]}
         RedCurve {keys [<at 0 0.007 RCIM_Cubic>,<at 0.6172 0.0044>]}
        }}))
     "Float (Alpha)"=&FloatFromCurve(FloatCurve=DataInterface() {propertyValues {Curve {keys [<at 0 0>,<at 0.2 1>]}}})
    )
    Particles.OutlineOffset=1
   ) {isSetParameters=true}
   &SubUVAnimation(
    "Sprite Renderer "=DataInterface() {propertyValues {SpriteRenderer=&self:NE_VC_Boost_Sprite_Stars_0.NiagaraSpriteRendererProperties_1}}
    UseEndFrame=false
    UseStartFrame=false
   )
   &SubUVAnimation(
    "Sprite Renderer "=DataInterface() {propertyValues {SpriteRenderer=&self:NE_VC_Boost_Sprite_Stars_0.NiagaraSpriteRendererProperties_2}}
    UseEndFrame=false
    UseStartFrame=false
   ) {module=SubUVAnimation001}
  ]
  rendererClasses [&NiagaraSpriteRendererProperties]
  renderers [
   {rendererClass=&NiagaraSpriteRendererProperties}
   {rendererClass=&NiagaraSpriteRendererProperties, rendererIndex=1}
   {rendererClass=&NiagaraSpriteRendererProperties, rendererIndex=2}
   {rendererClass=&NiagaraSpriteRendererProperties, rendererIndex=3}
  ]
 }
}
summary {
 systemName=NS_Boost_ToonPsplus
 emitters [
  {emitterName=NE_VC_Boost_Sprite_Stars, rendererClasses [&NiagaraSpriteRendererProperties]}
  {emitterName=NE_VC_Boost_Sprite_Parent, rendererClasses [&NiagaraSpriteRendererProperties]}
  {emitterName=NE_VC_Boost_Ribbon_Parent001, rendererClasses [&NiagaraRibbonRendererProperties]}
 ]
 userVariables [
  User.bIsFrontendPreview {type=NiagaraBool, defaultValue=true}
  User.bIsPainted {type=NiagaraBool}
  User.blsIsCoroColor {type=NiagaraBool}
  User.bPreviewVehicleSpeed {type=NiagaraBool}
  User.ColorVariantID {type=NiagaraInt32}
  User.DTCarOffset {type=NiagaraPosition, description="", defaultValue {}}
  User.GameModeID {type=NiagaraInt32}
  User.PaintedColor {type=LinearColor, description="", defaultValue {struct=LinearColor, value=<rgba 1 1 1 1>}}
  User.PaintedScalar {type=NiagaraFloat, defaultValue=1, description=""}
  User.PortCount {type=NiagaraInt32, defaultValue=2}
 ]
}
